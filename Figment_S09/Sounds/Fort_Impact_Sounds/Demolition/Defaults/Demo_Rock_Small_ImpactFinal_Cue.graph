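# Sound cue for the final impact of demolishing small rock, routed to the impacts sound class with small-melee attenuation and limited concurrency.
# A quality-level node picks between two layered mixes and two branches playing a single pitch-modulated optimized wave.
# Each mix stacks an immediate rock hit with three quieter, delayed, higher-pitched layers to form a staggered debris tail.
# The first mix draws on more random wave variations than the second.
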SoundCue /Figment_S09/Sounds/Fort_Impact_Sounds/Demolition/Defaults/Demo_Rock_Small_ImpactFinal_Cue
root=QualityLevel_0
nodes [
 {id=Delay_0, children [Modulator_4], properties {delayMax=0.2, delayMin=0.15}}
 {id=Delay_1, children [Modulator_0], properties {delayMax=0.2, delayMin=0.15}}
 {id=Delay_2, children [Modulator_6], properties {delayMax=0.08, delayMin=0.05}}
 {id=Delay_3, children [Modulator_7], properties {delayMax=0.52, delayMin=0.45}}
 {id=Delay_4, children [Modulator_10], properties {delayMax=0.08, delayMin=0.05}}
 {id=Delay_5, children [Modulator_11], properties {delayMax=0.52, delayMin=0.45}}
 {id=Mixer_0, children [Modulator_5,Delay_0,Delay_2,Delay_3], properties {inputVolume [1,0.7,0.5,0.4]}}
 {id=Mixer_3, children [Modulator_1,Delay_1,Delay_4,Delay_5], properties {inputVolume [1,0.7,0.5,0.4]}}
 {id=Modulator_0, children [Random_10], properties {pitchMax=1.5, pitchMin=1.4}}
 {id=Modulator_1, children [Random_0], properties {pitchMax=1.2}}
 {id=Modulator_10, children [Random_5], properties {pitchMax=1.5, pitchMin=1.3}}
 {id=Modulator_11, children [Random_5], properties {pitchMax=1.5, pitchMin=1.3}}
 {id=Modulator_2, children [WavePlayer_2], properties {pitchMax=1.1, pitchMin=0.9}}
 {id=Modulator_3, children [WavePlayer_3], properties {pitchMax=1.1, pitchMin=0.9}}
 {id=Modulator_4, children [Random_2], properties {pitchMax=1.5, pitchMin=1.4}}
 {id=Modulator_5, children [WavePlayer_11], properties {pitchMax=1.2}}
 {id=Modulator_6, children [Random_1], properties {pitchMax=1.5, pitchMin=1.3}}
 {id=Modulator_7, children [Random_1], properties {pitchMax=1.5, pitchMin=1.3}}
 {id=QualityLevel_0, children [Mixer_3,Mixer_0,Modulator_2,Modulator_3]}
 {id=Random_0, children [WavePlayer_1,WavePlayer_0], properties {weights [1,1]}}
 {id=Random_1, children [WavePlayer_7,WavePlayer_8], properties {weights [1,1]}}
 {id=Random_10, children [WavePlayer_55, WavePlayer_58, WavePlayer_59], properties {weights [1,1,1]}}
 {id=Random_2, children [WavePlayer_4,WavePlayer_5], properties {bRandomizeWithoutReplacement=false, weights [1,1]}}
 {id=Random_5, children [WavePlayer_63, WavePlayer_64, WavePlayer_65], properties {weights [1,1,1]}}
 {id=WavePlayer_0, properties {soundWaveAssetPtr=/Game/Sounds/Fort_Impact_Sounds/Demolition/Defaults/rock_impact_stereo_10}}
 {id=WavePlayer_1, properties {soundWaveAssetPtr=/Game/Sounds/Fort_Impact_Sounds/Demolition/Defaults/rock_impact_stereo_08}}
 {id=WavePlayer_11, properties {soundWaveAssetPtr=&rock_impact_stereo_08}}
 {id=WavePlayer_2, properties {soundWaveAssetPtr=/Game/Sounds/Fort_Impact_Sounds/Demolition/Defaults/Rock_Small_ImpactFinal_opt}}
 {id=WavePlayer_3, properties {soundWaveAssetPtr=&Rock_Small_ImpactFinal_opt}}
 {id=WavePlayer_4, properties {soundWaveAssetPtr=/Game/Sounds/Fort_Impact_Sounds/Demolition/Defaults/rock_impact_stereo_03}}
 {id=WavePlayer_5, properties {soundWaveAssetPtr=/Game/Sounds/Fort_Impact_Sounds/Demolition/Defaults/rock_impact_stereo_06}}
 {id=WavePlayer_55, properties {soundWaveAssetPtr=&rock_impact_stereo_03}}
 {id=WavePlayer_58, properties {soundWaveAssetPtr=&rock_impact_stereo_06}}
 {id=WavePlayer_59, properties {soundWaveAssetPtr=/Game/Sounds/Fort_Impact_Sounds/Demolition/Defaults/rock_impact_stereo_07}}
 {id=WavePlayer_63, properties {soundWaveAssetPtr=/Figment_S09/Sounds/Fort_Impact_Sounds/Demolition/Fort_Impact_RockFinal_01}}
 {id=WavePlayer_64, properties {soundWaveAssetPtr=/Figment_S09/Sounds/Fort_Impact_Sounds/Demolition/Fort_Impact_RockFinal_02}}
 {id=WavePlayer_65, properties {soundWaveAssetPtr=/Figment_S09/Sounds/Fort_Impact_Sounds/Demolition/Fort_Impact_RockFinal_03}}
 {id=WavePlayer_7, properties {soundWaveAssetPtr=&Fort_Impact_RockFinal_01}}
 {id=WavePlayer_8, properties {soundWaveAssetPtr=&Fort_Impact_RockFinal_02}}
]
properties {
 attenuationSettings=&Fort_Weapon_Melee_Sml_Att
 maxDistance=2800
 soundClassObject=&Fort_Impacts
 volumeMultiplier=0.5
 concurrencyOverrides {maxCount=3, resolutionRule=StopOldest}
 concurrencySet [&CONC_Object_Impacts_FourMax]
}
